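# Niagara skin effect for the Clash_V large male body cosmetic, adding animated ribbons and meshes that follow the character's skeleton.
# Two looping burst ribbon emitters use FN_BoneFlow to wrap strands around four bones, with curve-driven radius, width and fade.
# Two mesh emitters track calf and forearm bones, spawning via a movement-threshold rate with animated material parameters.
# Hand and foot hidden flags zero ribbon width and set mesh visibility tags.
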NiagaraSystem /BRCosmetics/Characters/Player/Male/Large/Bodies/M_LRG_Clash_V/FX/NS_Clash_V_Skin
dependencies {
 usedDataInterfaces [&NiagaraDataInterfaceCurve, &NiagaraDataInterfaceArrayFloat, &NiagaraDataInterfaceArrayFloat3]
 usedDynamicInputs [
  &NormalizeFloat
  &VectorLength
  &SetBoolByFloatComparison
 ]
 usedModules [
  &self:SystemSpawnScript.SystemScriptSource.SystemScriptGraph.NiagaraNodeAssignment_0.SetVariables_#1
  &FN_Integrator
  &SystemState
 ]
}
emitters {
 Bone_Flow_Ribbon01 {
  emitterUpdateScript [
   &EmitterState(
    Life Cycle Mode=::NewEnumerator1
    Loop Delay=&RandomRangeFloat(Minimum=0, Maximum=1)
    Loop Duration=&RandomRangeFloat(Minimum=0.5, Maximum=1)
    Recalculate Duration Each Loop=true
    UseLoopDelay=true
   )
   &SpawnBurst_Instantaneous(Spawn Count=20, Use Spawn Probability=false, UseLoopCountLimit=false)
  ]
  particleSpawnScript [
   &FN_SetSpawnGroup(Inclusivity=true, Normalize Spawn Group=true, Particles Per Group=20, Pattern=::NewEnumerator0, Set Ribbon ID and Ribbon Link Order=true)
   &InitializeRibbon(
    Lifetime Max=1.5
    Lifetime Min=1.5
    Lifetime Mode:ENiagara_LifetimeMode=::NewEnumerator1
    UsePositionOffset=false
    Write Material Random=true
   )
   &CameraOffset(Camera Offset Amount=10)
   &DynamicMaterialParameters(
    Index 0 Param 1=0.5
    Index 0 Param 2=1
    Index 0 Param 3=1
    Index 0 Param 4=1
    Module.Param0WriteEnabled=true
    Module.Param1WriteEnabled=true
    Module.Param2WriteEnabled=true
    Module.Param3WriteEnabled=true
   )
  ]
  particleUpdateScript [
   &ParticleState()
   &FN_BoneFlow(
    Angle:NiagaraFloat=&Multiply_Float(
     A=$Particles.Normalized Spawn Group
     B=&Add_Float(
      A=&Multiply_Float(A=$Particles.Age, B=0.5)
      B=&Lerp_Float(A=1.5, B=2, Alpha=$Particles.Normalized Spawn Group)
     )
    )
    BoneID:NiagaraInt32=&ModuloInt(A=$Particles.SpawnGroup, B=4)
    Height:NiagaraFloat=&Multiply_Float(A=$Particles.Normalized Spawn Group, B=1)
    Radius:NiagaraFloat=&Multiply_Float(
     A=&FloatFromCurve(
      FloatCurve=DataInterface() {propertyValues {Curve {keys [<at 0 3 RCIM_Cubic>,<at 0.7 13 RCIM_Cubic>,<at 1 3>]}}}
      CurveIndex=$Particles.Normalized Spawn Group
      Scale Curve=&RandomRangeFloat(Minimum=0.7, Maximum=0.7)
     )
     B=&FloatFromCurve(
      FloatCurve=DataInterface() {propertyValues {Curve {keys [
          {interpMode=RCIM_Cubic, tangentMode=RCTM_User, time=0, value=0}
          {interpMode=RCIM_Cubic, tangentMode=RCTM_User, time=0.7, value=1}
          {tangentMode=RCTM_User, time=1, value=0.3}
         ]}}}
      Scale Curve=&RandomRangeFloat(Minimum=0.9, Maximum=1)
     )
    )
    SkeletalMesh:NiagaraDataInterfaceSkeletalMesh=$User.SK_Mesh
   )
   /SetVariables(Particles.RibbonFacing=$Output.FN_BoneFlow.SplineNormal) {isSetParameters=true, module=SetVariables_#2}
   &ScaleRibbonWidth(Ribbon Width Scale=&Multiply_Float(
     A=&Multiply_Float(
      A=&Multiply_Float(
       A=&FloatFromCurve(
        FloatCurve=DataInterface() {propertyValues {Curve {keys [
            {interpMode=RCIM_Cubic, tangentMode=RCTM_User, time=0, value=0}
            <at 0.1 0>
            {interpMode=RCIM_Cubic, tangentMode=RCTM_Break, time=0.2, value=1}
            {arriveTangent=-1.5728, interpMode=RCIM_Cubic, leaveTangent=-1.5728, tangentMode=RCTM_User, time=1, value=0}
           ]}}}
        Scale Curve=&RandomRangeFloat(Minimum=0.7, Maximum=1)
       )
       B=1
      )
      B=&FloatFromCurve(
       FloatCurve=DataInterface() {propertyValues {Curve {keys [
           {interpMode=RCIM_Cubic, tangentMode=RCTM_User, time=0, value=0}
           {interpMode=RCIM_Cubic, tangentMode=RCTM_User, time=0.3, value=1}
           {arriveTangent=-2.7836, interpMode=RCIM_Cubic, leaveTangent=-2.7836, tangentMode=RCTM_User, time=0.859, value=0.2872}
           {arriveTangent=-1.923, interpMode=RCIM_Cubic, leaveTangent=-1.923, tangentMode=RCTM_User, time=1, value=0}
          ]}}}
       CurveIndex=$Particles.Normalized Spawn Group
       Scale Curve=&FloatFromCurve(
        FloatCurve=DataInterface() {propertyValues {Curve {keys [
            {interpMode=RCIM_Cubic, tangentMode=RCTM_User, time=0, value=1}
            {interpMode=RCIM_Cubic, tangentMode=RCTM_User, time=0.3, value=0.4}
            {interpMode=RCIM_Cubic, tangentMode=RCTM_User, time=0.5, value=1.5}
            {interpMode=RCIM_Cubic, tangentMode=RCTM_User, time=0.7, value=0.4}
            {interpMode=RCIM_Cubic, tangentMode=RCTM_User, time=1, value=1}
           ]}}}
        CurveIndex=&Multiply_Float(
         A=$Particles.Normalized Spawn Group
         B=&Multiply_Float(A=$Particles.Age, B=&RandomRangeFloat(Minimum=1, Maximum=1.5))
        )
       )
      )
     )
     B=&SelectFloatFromArray(
      Direct Array Index=&self:PartHiddenBoneFlowCheck(
       is Hand Hidden:NiagaraBool=$User.bisHandHidden
       is Foot Hidden:NiagaraBool=$User.bisFootHidden
       Selection:NiagaraInt32=&ModuloInt(A=$Particles.SpawnGroup, B=4)
      )
      Float Selection Array=DataInterface() {propertyValues {FloatData [1,0]}}
     )
    ))
   &ScaleColor(
    Scale Alpha=&FloatFromCurve(FloatCurve=DataInterface() {propertyValues {Curve {keys [
         {interpMode=RCIM_Cubic, tangentMode=RCTM_User, time=0, value=0}
         <at 0.2 1>
         {interpMode=RCIM_Cubic, tangentMode=RCTM_User, time=0.8, value=1}
         {tangentMode=RCTM_User, time=1, value=0}
        ]}}})
    ScaleA=true
    ScaleRGB=true
   )
   &SolveForcesAndVelocity(Clamp Velocity=false, Limit Acceleration=false)
  ]
  rendererClasses [&NiagaraRibbonRendererProperties]
  renderers [{rendererClass=&NiagaraRibbonRendererProperties}]
 }
 Bone_Flow_Ribbon02 {
  emitterUpdateScript [
   &EmitterState(
    Life Cycle Mode=::NewEnumerator1
    Loop Delay=&RandomRangeFloat(Minimum=0.5, Maximum=2)
    Loop Duration=&RandomRangeFloat(Minimum=1.5, Maximum=2)
    Recalculate Duration Each Loop=true
    UseLoopDelay=true
   )
   &SpawnBurst_Instantaneous(Spawn Count=15, Use Spawn Probability=false, UseLoopCountLimit=false)
  ]
  particleSpawnScript [
   &FN_SetSpawnGroup(Inclusivity=true, Normalize Spawn Group=true, Particles Per Group=15, Pattern=::NewEnumerator0, Set Ribbon ID and Ribbon Link Order=true)
   &InitializeRibbon(Lifetime Min=2, Lifetime Mode:ENiagara_LifetimeMode=::NewEnumerator1, Ribbon Width=8, UsePositionOffset=false, Write Material Random=true)
   &CameraOffset(Camera Offset Amount=10)
   &DynamicMaterialParameters(
    Index 0 Param 1=0.5
    Index 0 Param 2=1
    Index 0 Param 3=1
    Index 0 Param 4=1
    Module.Param0WriteEnabled=true
    Module.Param1WriteEnabled=true
    Module.Param2WriteEnabled=true
    Module.Param3WriteEnabled=true
   )
  ]
  particleUpdateScript [
   &ParticleState()
   &FN_BoneFlow(
    Angle:NiagaraFloat=&Multiply_Float(
     A=$Particles.Normalized Spawn Group
     B=&Add_Float(
      A=&Multiply_Float(A=$Particles.Age, B=0.5)
      B=&Lerp_Float(A=0.5, B=1.5, Alpha=$Particles.Normalized Spawn Group)
     )
    )
    BoneID:NiagaraInt32=&ModuloInt(A=$Particles.SpawnGroup, B=4)
    Height:NiagaraFloat=&Multiply_Float(A=$Particles.Normalized Spawn Group, B=1)
    Radius:NiagaraFloat=&Multiply_Float(
     A=&FloatFromCurve(
      FloatCurve=DataInterface() {propertyValues {Curve {keys [<at 0 3 RCIM_Cubic>,<at 0.7 13 RCIM_Cubic>,<at 1 3>]}}}
      CurveIndex=$Particles.Normalized Spawn Group
      Scale Curve=&RandomRangeFloat(Minimum=0.7, Maximum=0.7)
     )
     B=&FloatFromCurve(
      FloatCurve=DataInterface() {propertyValues {Curve {keys [
          {interpMode=RCIM_Cubic, tangentMode=RCTM_User, time=0, value=0}
          {interpMode=RCIM_Cubic, tangentMode=RCTM_User, time=0.7, value=1}
          {tangentMode=RCTM_User, time=1, value=0.3}
         ]}}}
      Scale Curve=&RandomRangeFloat(Minimum=1.1, Maximum=1)
     )
    )
    SkeletalMesh:NiagaraDataInterfaceSkeletalMesh=$User.SK_Mesh
   )
   /SetVariables(Particles.RibbonFacing=$Output.FN_BoneFlow.SplineNormal) {isSetParameters=true, module=SetVariables_#2}
   &ScaleRibbonWidth(Ribbon Width Scale=&Multiply_Float(
     A=&Multiply_Float(
      A=&Multiply_Float(
       A=&FloatFromCurve(
        FloatCurve=DataInterface() {propertyValues {Curve {keys [
            {interpMode=RCIM_Cubic, tangentMode=RCTM_User, time=0, value=0}
            <at 0.1 0>
            {interpMode=RCIM_Cubic, tangentMode=RCTM_Break, time=0.2, value=1}
            {arriveTangent=-1.5728, interpMode=RCIM_Cubic, leaveTangent=-1.5728, tangentMode=RCTM_User, time=1, value=0}
           ]}}}
        Scale Curve=&RandomRangeFloat(Minimum=0.7, Maximum=1)
       )
       B=1
      )
      B=&FloatFromCurve(
       FloatCurve=DataInterface() {propertyValues {Curve {keys [
           {interpMode=RCIM_Cubic, tangentMode=RCTM_User, time=0, value=0}
           {interpMode=RCIM_Cubic, tangentMode=RCTM_User, time=0.3, value=1}
           {arriveTangent=-2.7836, interpMode=RCIM_Cubic, leaveTangent=-2.7836, tangentMode=RCTM_User, time=0.859, value=0.2872}
           {arriveTangent=-1.923, interpMode=RCIM_Cubic, leaveTangent=-1.923, tangentMode=RCTM_User, time=1, value=0}
          ]}}}
       CurveIndex=$Particles.Normalized Spawn Group
       Scale Curve=&FloatFromCurve(
        FloatCurve=DataInterface() {propertyValues {Curve {keys [
            {interpMode=RCIM_Cubic, tangentMode=RCTM_User, time=0, value=1}
            {interpMode=RCIM_Cubic, tangentMode=RCTM_User, time=0.3, value=0.4}
            {interpMode=RCIM_Cubic, tangentMode=RCTM_User, time=0.5, value=1.5}
            {interpMode=RCIM_Cubic, tangentMode=RCTM_User, time=0.7, value=0.4}
            {interpMode=RCIM_Cubic, tangentMode=RCTM_User, time=1, value=1}
           ]}}}
        CurveIndex=&Multiply_Float(
         A=$Particles.Normalized Spawn Group
         B=&Multiply_Float(A=$Particles.Age, B=&RandomRangeFloat(Minimum=1, Maximum=1.5))
        )
       )
      )
     )
     B=&SelectFloatFromArray(
      Direct Array Index=&self:PartHiddenBoneFlowCheck(
       is Hand Hidden:NiagaraBool=$User.bisHandHidden
       is Foot Hidden:NiagaraBool=$User.bisFootHidden
       Selection:NiagaraInt32=&ModuloInt(A=$Particles.SpawnGroup, B=4)
      )
      Float Selection Array=DataInterface() {propertyValues {FloatData [1,0]}}
     )
    ))
   &ScaleColor(
    Scale Alpha=&FloatFromCurve(FloatCurve=DataInterface() {propertyValues {Curve {keys [
         {interpMode=RCIM_Cubic, tangentMode=RCTM_User, time=0, value=0}
         <at 0.2 1>
         {interpMode=RCIM_Cubic, tangentMode=RCTM_User, time=0.8, value=1}
         {tangentMode=RCTM_User, time=1, value=0}
        ]}}})
    ScaleA=true
    ScaleRGB=true
   )
   &SolveForcesAndVelocity(Clamp Velocity=false, Limit Acceleration=false)
  ]
  rendererClasses [&NiagaraRibbonRendererProperties]
  renderers [{rendererClass=&NiagaraRibbonRendererProperties}]
 }
 Mesh_Calf {
  emitterUpdateScript [
   &EmitterState(Inactive Response=::NewEnumerator1, Life Cycle Mode=::NewEnumerator1, Loop Duration=4, UseLoopDelay=false)
   &self:ThresholdSpawnRate(MovementThreshold=50, SpawnRate=0.25) {module=ThresholdSpawnRate}
   &SpawnRate(SpawnRate=$Emitter.ThresholdSpawnRate, Use Spawn Probability=false)
  ]
  particleSpawnScript [
   &FN_SetSpawnGroup(Number of Spawn Groups=2)
   &InitializeParticle(
    Lifetime Max=2.2
    Lifetime Min=2.5
    Lifetime Mode=::NewEnumerator1
    Mesh Scale=&RandomRangeVector(Minimum=<0.8 0.65 0.65>, Maximum=<0.7 0.65 0.65>)
    Mesh Scale Mode=::NewEnumerator4
    UsePositionOffset=false
    Write Material Random=true
   )
   &InitialMeshOrientation(
    Orientation Axis=&SelectVectorFromArray(Direct Array Index=$Particles.SpawnGroup, Vector Selection Array=DataInterface() {propertyValues {FloatData [<-1 0 0>, <1 0 0>]}})
    Rotation=&SelectVectorFromArray(Direct Array Index=$Particles.SpawnGroup, Vector Selection Array=DataInterface() {propertyValues {FloatData [<0 0 0>, <0 0 0>]}})
    Use Orientation Vector=true
    Use Rotation Vector=true
   )
   &CameraOffset(Camera Offset Amount=2)
  ]
  particleUpdateScript [
   &ParticleState()
   &UpdateMeshOrientation(Rotation Rate=-0.1, Rotation Vector=<1 0 0>)
   &SkeletalMeshLocation(
    Bone / Socket Index=$Particles.SpawnGroup
    Bone Sampling Mode=::NewEnumerator4
    "Interpolate Mesh Sampling (Update)"=true
    MaskSpawnGroup=false
    Sampled Position Offset=&SelectVectorFromArray(
     Direct Array Index=$Particles.SpawnGroup
     Vector Selection Array=DataInterface() {propertyValues {MaxElements=2, FloatData [<0 -5 0>, <0 5 0>]}}
    )
    Skeletal Mesh=DataInterface() {propertyValues {
      DefaultMesh=null
      ExcludeBoneName=root
      OptionalComponentClass=null
      PreviewMesh=/BRCosmetics/Characters/Player/Male/Large/Bodies/M_LRG_Clash_V/Meshes/M_LRG_Clash_V
      SkinningMode=SkinOnTheFly
      SoftSourceActor=null
      SourceMode=Default
      UvSetIndex=0
      WholeMeshLOD=-1
      bExcludeBone=false
      bReadDeformedGeometry=true
      bRequireCurrentFrameData=true
      ComponentTags []
      FilteredBones [calf_twist_02_l, calf_twist_02_r]
      FilteredSockets []
      MeshUserParameter {parameter {name=None, typeDefHandle {registeredTypeIndex=1}, varData []}}
      SamplingRegions []
     }}
   )
   &DynamicMaterialParameters(
    Index 0 Param 1=&FloatFromCurve(FloatCurve=DataInterface() {propertyValues {Curve {keys [<at 0 1>,<at 0.5 0.5>]}}})
    Index 0 Param 2=&FloatFromCurve(FloatCurve=DataInterface() {propertyValues {Curve {keys [
         {interpMode=RCIM_Cubic, leaveTangent=-2.9591, tangentMode=RCTM_Break, time=0, value=1.5}
         {arriveTangent=-0.3114, interpMode=RCIM_Cubic, tangentMode=RCTM_Break, time=1, value=0}
        ]}}})
    Index 1 Param 1=&FloatFromCurve(FloatCurve=DataInterface() {propertyValues {Curve {keys [
         {arriveTangent=-10.7468, interpMode=RCIM_Cubic, leaveTangent=-10.7468, tangentMode=RCTM_User, time=0, value=2}
         {interpMode=RCIM_Cubic, leaveTangent=0.5305, tangentMode=RCTM_Break, time=0.1, value=1}
         {interpMode=RCIM_Cubic, tangentMode=RCTM_Break, time=1, value=0.2}
        ]}}})
    Index 1 Param 2=2
    Index 1 Param 3=&FloatFromCurve(FloatCurve=DataInterface() {propertyValues {Curve {keys [
         {interpMode=RCIM_Cubic, leaveTangent=10.6169, tangentMode=RCTM_Break, time=0, value=-0.2}
         {arriveTangent=2.4889, interpMode=RCIM_Cubic, leaveTangent=0.1906, tangentMode=RCTM_Break, time=0.1, value=0.5}
         {arriveTangent=-0.3416, interpMode=RCIM_Cubic, leaveTangent=-1.8967, tangentMode=RCTM_Break, time=0.6, value=0.5}
         {arriveTangent=0.3626, interpMode=RCIM_Cubic, tangentMode=RCTM_Break, time=0.9, value=0}
        ]}}})
    Module.Param0WriteEnabled=true
    Module.Param0WriteEnabled1=true
    Module.Param1WriteEnabled=true
    Module.Param1WriteEnabled1=true
    Module.Param2WriteEnabled=false
    Module.Param2WriteEnabled1=true
    Module.Param3WriteEnabled=false
    Module.Param3WriteEnabled1=false
    Write Parameter Index 1=true
   )
   &ScaleMeshSize(Scale Factor=&MakeVector(
     X=&FloatFromCurve(FloatCurve=DataInterface() {propertyValues {Curve {keys [<at 0.65 1>,<at 1 0.8>]}}})
     Y=&FloatFromCurve(FloatCurve=DataInterface() {propertyValues {Curve {keys [
          {interpMode=RCIM_Cubic, tangentMode=RCTM_User, time=0, value=0.7}
          {interpMode=RCIM_Cubic, tangentMode=RCTM_User, time=0.6, value=1}
          {arriveTangent=-4.4791, interpMode=RCIM_Cubic, leaveTangent=-4.4791, tangentMode=RCTM_User, time=1, value=0}
         ]}}})
     Z=&FloatFromCurve(FloatCurve=DataInterface() {propertyValues {Curve {keys [
          {interpMode=RCIM_Cubic, tangentMode=RCTM_User, time=0, value=0.7}
          {interpMode=RCIM_Cubic, tangentMode=RCTM_User, time=0.6, value=1}
          {arriveTangent=-6.148, interpMode=RCIM_Cubic, leaveTangent=-6.148, tangentMode=RCTM_User, time=1, value=0}
         ]}}})
    ))
   /SetVariables(Particles.VisibilityTag=&MakeCustomIntFromBool(Bool=$User.bisFootHidden, True INT=1, False INT=0)) {isSetParameters=true, module=SetVariables_#3}
  ]
  rendererClasses [&NiagaraMeshRendererProperties]
  renderers [{rendererClass=&NiagaraMeshRendererProperties, rendererIndex=0}]
 }
 Mesh_Lowerarm {
  emitterUpdateScript [
   &EmitterState(Inactive Response=::NewEnumerator1, Life Cycle Mode=::NewEnumerator1, Loop Duration=4, UseLoopDelay=false)
   &self:ThresholdSpawnRate(MovementThreshold=50, SpawnRate=0.25) {module=ThresholdSpawnRate}
   &SpawnRate(SpawnRate=$Emitter.ThresholdSpawnRate, Use Spawn Probability=false)
  ]
  particleSpawnScript [
   &FN_SetSpawnGroup(Number of Spawn Groups=2)
   &InitializeParticle(
    Lifetime Max=2.3
    Lifetime Min=2
    Lifetime Mode=::NewEnumerator1
    Mesh Scale=&RandomRangeVector(Minimum=<0.75 0.4 0.4>, Maximum=<0.8 0.4 0.4>)
    Mesh Scale Mode=::NewEnumerator4
    UsePositionOffset=false
    Write Material Random=true
   )
   &InitialMeshOrientation(
    Orientation Axis=&SelectVectorFromArray(Direct Array Index=$Particles.SpawnGroup, Vector Selection Array=DataInterface() {propertyValues {FloatData [<-1 0 0>, <1 0 0>]}})
    Rotation=&RandomRangeVector(Minimum=<1 0 0>, Maximum=<0 0 0>)
    Rotation Coordinate Space=::NewEnumerator2
    Use Orientation Vector=true
    Use Rotation Vector=true
   )
   &CameraOffset(Camera Offset Amount=2)
  ]
  particleUpdateScript [
   &ParticleState()
   &UpdateMeshOrientation(Rotation Rate=-0.1, Rotation Vector=<1 0 0>)
   &SkeletalMeshLocation(
    Bone / Socket Index=$Particles.SpawnGroup
    Bone Sampling Mode=::NewEnumerator4
    "Interpolate Mesh Sampling (Update)"=true
    MaskSpawnGroup=false
    Sampled Position Offset=&SelectVectorFromArray(
     Direct Array Index=$Particles.SpawnGroup
     Vector Selection Array=DataInterface() {propertyValues {MaxElements=2, FloatData [<15 0 0>, <-15 0 0>]}}
    )
    Skeletal Mesh=DataInterface() {propertyValues {
      DefaultMesh=null
      ExcludeBoneName=root
      OptionalComponentClass=null
      PreviewMesh=&M_LRG_Clash_V
      SkinningMode=SkinOnTheFly
      SoftSourceActor=null
      SourceMode=Default
      UvSetIndex=0
      WholeMeshLOD=-1
      bExcludeBone=false
      bReadDeformedGeometry=true
      bRequireCurrentFrameData=true
      ComponentTags []
      FilteredBones [lowerarm_l,lowerarm_r]
      FilteredSockets []
      MeshUserParameter {parameter {name=None, typeDefHandle {registeredTypeIndex=1}, varData []}}
      SamplingRegions []
     }}
   )
   &DynamicMaterialParameters(
    Index 0 Param 1=&FloatFromCurve(FloatCurve=DataInterface() {propertyValues {Curve {keys [<at 0 1>,<at 0.5 0.5>]}}})
    Index 0 Param 2=&FloatFromCurve(FloatCurve=DataInterface() {propertyValues {Curve {keys [
         {interpMode=RCIM_Cubic, leaveTangent=-2.9591, tangentMode=RCTM_Break, time=0, value=1.5}
         {arriveTangent=-0.3114, interpMode=RCIM_Cubic, tangentMode=RCTM_Break, time=1, value=0}
        ]}}})
    Index 1 Param 1=&FloatFromCurve(FloatCurve=DataInterface() {propertyValues {Curve {keys [
         {arriveTangent=-10.7468, interpMode=RCIM_Cubic, leaveTangent=-10.7468, tangentMode=RCTM_User, time=0, value=2}
         {interpMode=RCIM_Cubic, leaveTangent=0.5305, tangentMode=RCTM_Break, time=0.1, value=1}
         {interpMode=RCIM_Cubic, tangentMode=RCTM_Break, time=1, value=0.2}
        ]}}})
    Index 1 Param 2=2
    Index 1 Param 3=&FloatFromCurve(FloatCurve=DataInterface() {propertyValues {Curve {keys [
         {interpMode=RCIM_Cubic, leaveTangent=10.6169, tangentMode=RCTM_Break, time=0, value=-0.2}
         {arriveTangent=2.4889, interpMode=RCIM_Cubic, leaveTangent=0.1906, tangentMode=RCTM_Break, time=0.1, value=0.5}
         {arriveTangent=-0.3416, interpMode=RCIM_Cubic, leaveTangent=-1.8967, tangentMode=RCTM_Break, time=0.6, value=0.5}
         {arriveTangent=0.3626, interpMode=RCIM_Cubic, tangentMode=RCTM_Break, time=0.9, value=0}
        ]}}})
    Module.Param0WriteEnabled=true
    Module.Param0WriteEnabled1=true
    Module.Param1WriteEnabled=true
    Module.Param1WriteEnabled1=true
    Module.Param2WriteEnabled=false
    Module.Param2WriteEnabled1=true
    Module.Param3WriteEnabled=false
    Module.Param3WriteEnabled1=false
    Write Parameter Index 1=true
   )
   &ScaleMeshSize(Scale Factor=&MakeVector(
     X=&FloatFromCurve(FloatCurve=DataInterface() {propertyValues {Curve {keys [<at 0.65 1>,<at 1 0.8>]}}})
     Y=&FloatFromCurve(FloatCurve=DataInterface() {propertyValues {Curve {keys [
          {interpMode=RCIM_Cubic, tangentMode=RCTM_User, time=0, value=0.7}
          {interpMode=RCIM_Cubic, tangentMode=RCTM_User, time=0.6, value=1}
          {arriveTangent=-4.4791, interpMode=RCIM_Cubic, leaveTangent=-4.4791, tangentMode=RCTM_User, time=1, value=0}
         ]}}})
     Z=&FloatFromCurve(FloatCurve=DataInterface() {propertyValues {Curve {keys [
          {interpMode=RCIM_Cubic, tangentMode=RCTM_User, time=0, value=0.7}
          {interpMode=RCIM_Cubic, tangentMode=RCTM_User, time=0.6, value=1}
          {arriveTangent=-6.148, interpMode=RCIM_Cubic, leaveTangent=-6.148, tangentMode=RCTM_User, time=1, value=0}
         ]}}})
    ))
   /SetVariables(Particles.VisibilityTag=&MakeCustomIntFromBool(Bool=$User.bisHandHidden, True INT=1, False INT=0)) {isSetParameters=true, module=SetVariables_#3}
  ]
  rendererClasses [&NiagaraMeshRendererProperties]
  renderers [{rendererClass=&NiagaraMeshRendererProperties, rendererIndex=0}]
 }
}
summary {
 systemName=NS_Clash_V_Skin
 emitters [
  {emitterName=Bone_Flow_Ribbon01, rendererClasses [&NiagaraRibbonRendererProperties]}
  {emitterName=Bone_Flow_Ribbon02, rendererClasses [&NiagaraRibbonRendererProperties]}
  {emitterName=Mesh_Calf, rendererClasses [&NiagaraMeshRendererProperties]}
  {emitterName=Mesh_Lowerarm, rendererClasses [&NiagaraMeshRendererProperties]}
 ]
 userVariables [
  User.Big_Mesh02Tendrils_Size {type=NiagaraFloat, defaultValue=0, description=""}
  User.Big_MeshTendrils_Dissolve {type=NiagaraFloat, defaultValue=0, description=""}
  User.Big_MeshTendrils_Location {type=NiagaraFloat, defaultValue=0, description=""}
  User.Big_MeshTendrils_Size {type=NiagaraFloat, defaultValue=0, description=""}
  User.bisFootHidden {type=NiagaraBool}
  User.bisHandHidden {type=NiagaraBool}
  User.Is_SkinFX_Trigger {type=NiagaraBool}
  User.IsSkydiving {type=NiagaraBool}
  User.Kill {type=NiagaraBool}
  User.RibbonIDOffset {type=NiagaraInt32}
  User.SK_Mesh {
   type=NiagaraDataInterfaceSkeletalMesh
   defaultValue {
    struct=NiagaraExt_VariableValue_DataInterface
    value {dataInterface=&self:NiagaraDataInterfaceSkeletalMesh_0}
   }
  }
  User.SK_Mesh_Tendrils {
   type=NiagaraDataInterfaceSkeletalMesh
   defaultValue {
    struct=NiagaraExt_VariableValue_DataInterface
    value {dataInterface=&self:NiagaraDataInterfaceSkeletalMesh_2}
   }
  }
  User.SkeletalMesh {
   type=NiagaraDataInterfaceSkeletalMesh
   defaultValue {
    struct=NiagaraExt_VariableValue_DataInterface
    value {dataInterface=&self:NiagaraDataInterfaceSkeletalMesh_3}
   }
  }
  User.SkeletalMesh_Trail_Scoket {
   type=NiagaraDataInterfaceSkeletalMesh
   defaultValue {
    struct=NiagaraExt_VariableValue_DataInterface
    value {dataInterface=&self:NiagaraDataInterfaceSkeletalMesh_6}
   }
  }
  User.Small_MeshTendrils_Burst {type=NiagaraInt32}
  User.Small_MeshTendrils_Location {type=NiagaraFloat, defaultValue=0, description=""}
  User.Small_MeshTendrils_Size {type=NiagaraFloat, defaultValue=0, description=""}
  User.Spikes_number {type=NiagaraFloat, defaultValue=20, description=""}
 ]
}
