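# Material function for the Clash character body that applies a curve-driven inflate, wobble and masked transition effect to incoming material attributes.
# World position offset pushes vertices along normals, combining sphere-masked inflation with panning world-aligned noise, scaled by MaterialCurveC and function inputs.
# A second sphere mask, moved by MaterialCurveA, sets the opacity mask for the transition.
# Two-sided sign darkens interior faces, and Faux's MF_bend_WPO feeds only the blended normal.
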MaterialFunction /BRCosmetics/Characters/Player/Male/Large/Bodies/M_LRG_Clash_V/Materials/MF_Clash
graph {nodes [
  n1 = GetMaterialAttributes(None=FunctionInput() {inputType=FunctionInput_MaterialAttributes}) {attributeGetTypes [#1, #2, #3]}
  n2 = TwoSidedSign()
  n3 = ScalarParameter() {desc=SphereMask Radius, group=Faux_WPO, parameterName=MaterialCurveC}
  n4 = LinearInterpolate(B=ScalarParameter() {defaultValue=0.2, parameterName=overallINfl}, Alpha=@n3)
  n5 = ScalarParameter() {defaultValue=0.8, parameterName=Radius}
  n6 = NamedRerouteUsage()
  n7 = NamedRerouteUsage()
  n8 = SphereMask(
   A=LinearInterpolate(B=Add(B=@n5) {constA=1}, Alpha=LinearInterpolate(Alpha=@n6) {constA=0.4, constB=0})
   B=@n7
   Radius=@n5
   Hardness=Constant()
  ) {hardnessPercent=0}
  n9 = Multiply(A=@n8, B=LinearInterpolate(B=ScalarParameter() {defaultValue=5, parameterName=Inflation Amount}, Alpha=@n3)) {constB=10, desc=spheremask inflation}
  n10 = Multiply(A=VertexNormalWS(), B=Multiply(A=Add(A=@n4, B=@n9), B=@n3)) {constB=5}
  n11 = TextureObjectParameter() {
   group=SuperLevelling_Mask
   parameterName=FX_WobbleNoise
   texture=/Game/Characters/Player/Male/Medium/Bodies/M_MED_Viper/Textures/Glitch/T_SoftNoise
  }
  n12 = NamedRerouteUsage()
  n13 = Multiply(A=Time(), B=VectorParameter() {parameterName=WobbleNoisePanSpeedAndDirection, defaultValue=<rgba 15 15 15 1>}.RGB)
  n14 = Power(Base=MaterialFunctionCall(
    TextureObject=@n11
    TextureSize=VectorParameter() {group=SuperLevelling_Mask, parameterName=FX_WobbleNoise_Tiling, defaultValue=<rgba 80 80 80 1>}.RGB
    WorldPosition=Add(A=@n12, B=@n13)
    World Space Normal=PreSkinnedNormal()
   ) {materialFunction=/Game/Weapons/WeaponSkins/Wraps/MaterialFunctions/WorldAlignedTexture}.XYZ Texture)
  n15 = VectorParameter() {parameterName=WobbleIntensity, defaultValue=<rgba 1.2 1.2 1.2 1>}
  n16 = Multiply(A=LinearInterpolate(A=@n15.RGB, B=Multiply(A=@n15.RGB) {constB=8}, Alpha=@n8), B=@n3)
  n17 = Multiply(
   A=Multiply(A=Multiply(A=@n14, B=@n16), B=VertexNormalWS())
   B=FunctionInput() {bUsePreviewValueAsDefault=true, inputName=thickness, inputType=FunctionInput_Scalar}
  )
  n18 = Multiply(A=NamedRerouteUsage()) {constB=110}
  n19 = NamedRerouteUsage()
  n20 = VertexNormalWS()
  n21 = TextureCoordinate() {coordinateIndex=1, desc=coord 1 has the split uvs}
  n22 = Normalize(VectorInput=Add(
    A=MaterialFunctionCall!(angle=@n18, Min=@n19, PreskinnedPosition=@n20, TexCoord=@n21) {materialFunction=/Game/Characters/Player/Male/Medium/Bodies/M_MED_Faux/Materials/MF_bend_WPO}.Result
    B=@n20
   ))
  n23 = ScalarParameter() {defaultValue=1.33, parameterName=CenterOfTrans}
  n24 = ScalarParameter() {desc=drives angle of bend - normalized, group=Faux_WPO, parameterName=MaterialCurveA}
  n25 = SetMaterialAttributes!(
   MaterialAttributes=@n1.MaterialAttributes
   BaseColor=LinearInterpolate(
    A=VectorParameter() {parameterName=InsideDiffuse Color, defaultValue=<rgba 0.0156 0.0089 0.0139 1>}.RGB
    B=@n1.BaseColor
    Alpha=Clamp(None=@n2)
   )
   WorldPositionOffset=Multiply(A=Add(A=@n10, B=@n17), B=FunctionInput() {bUsePreviewValueAsDefault=true, inputName=WPO, inputType=FunctionInput_Scalar})
   OpacityMask=SphereMask(
    A=@n7
    B=LinearInterpolate(A=@n23, Alpha=@n24) {constB=0.5}
    Radius=@n6
    Hardness=ScalarParameter() {defaultValue=1.5, parameterName=TransitionPower}
   )
   EmissiveColor
   Normal=MaterialFunctionCall(BaseNormal=Transform(None=@n22) {transformSourceType=TRANSFORMSOURCE_World, transformType=TRANSFORM_Tangent}, AdditionalNormal=@n1.Normal) {materialFunction=&BlendAngleCorrectedNormals}.Result
   PixelDepthOffset
  ) {attributeSetTypes [
    #3
    #4
    #5
    #1
    #2
    #6
   ]}
  FunctionOutput(None=@n25)
  NamedRerouteDeclaration(None=@n24) {name=BendAmt, nodeColor=<rgba 0.2656 1 0 1>}
  NamedRerouteDeclaration(None=LinearInterpolate(Alpha=ScalarParameter() {defaultValue=0.5, desc=zipperHeight, group=Faux_WPO, parameterName=MaterialCurveB}) {constB=1.5}) {name=zipperHeight, nodeColor=<rgba 0 1 0.6719 1>}
  n26 = PreSkinnedPosition()
  n27 = VertexInterpolator(VS=@n26)
  NamedRerouteDeclaration(None=FunctionInput(Preview=ShaderStageSwitch(PixelShader=@n27.PS, VertexShader=@n26)) {bUsePreviewValueAsDefault=true, inputName=PreskinnedPosition}) {name=PreskinnedPos, nodeColor=<rgba 1 0 0.3281 1>}
  NamedRerouteDeclaration(None=OneMinus(None=FunctionInput() {inputName=Gradient, inputType=FunctionInput_Scalar})) {name=Gradient, nodeColor=<rgba 0 0.2969 1 1>}
  NamedRerouteUsage()
 ]}
