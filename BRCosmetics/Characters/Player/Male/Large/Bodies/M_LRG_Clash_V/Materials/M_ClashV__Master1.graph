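# Master skinned-character material for the large male Clash V body, layering skin effects between the shared character PreFX and PostFX functions through material attributes.
# The SkinFX mask texture drives each layer: green feeds MF_Throbbing borrowed from Vertigo, red feeds MF_Clash, and blue blends the weapon-skin Clearcoat function.
# A static Transition switch includes or bypasses MF_Clash, with pre-skinned local position fed to both effect functions.
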Material /BRCosmetics/Characters/Player/Male/Large/Bodies/M_LRG_Clash_V/Materials/M_ClashV__Master1
graph {
 properties {
  bUseMaterialAttributes=true
  bUsedWithMeshDeformer=true
  bUsedWithMorphTargets=true
  bUsedWithSkeletalMesh=true
  ditherOpacityMask=true
  pixelDepthOffsetMode=PDOM_Legacy
  shadingModel=MSM_FromMaterialExpression
  twoSided=true
  usedShadingModels=Default Lit | Clear Coat | Subsurface Profile
 }
 nodes [
  Constant() {r=1}
  Constant() {r=1}
  n1 = TextureSampleParameter2D() {parameterName=SkinFX, samplerType=SAMPLERTYPE_Masks, texture=&T_BlackMask}
  n2 = MaterialFunctionCall() {materialFunction=&BaseFN_CharacterFunction_PreFX}
  n3 = MaterialFunctionCall!(Alpha=@n1.G, Pre SKinned Local=@n2.Pre skinned Local Position, ThicknessMap=@n1.G, In=@n2.Pre skinned Local Position) {materialFunction=/Game/Characters/Player/Male/Large/Bodies/M_LRG_Vertigo/Materials/MF_Throbbing}
  n4 = MaterialFunctionCall!(PreskinnedPosition=@n2.Pre skinned Local Position, thickness=@n1.R, Gradient=@n1.R, WPO=@n1.R, In=@n3.Result) {materialFunction=/Game/Characters/Player/Male/Large/Bodies/M_LRG_Clash_V/Materials/MF_Clash}
  MaterialFunctionCall(Material=MaterialFunctionCall!(Alpha=@n1.B, In=StaticSwitchParameter(True=@n4.Result, False=@n3.Result) {parameterName=Transition}) {materialFunction=/Game/Weapons/WeaponSkins/MaterialFunction/Clearcoat}.Result) {materialFunction=&BaseFN_CharacterFunction_PostFX}
 ]
}
