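# Control Rig for the M_LRG_Clash_V large male player body that drives corrective deform bones at elbows, knees, shoulders and hips.
# One execution chain reads each joint's parent-relative Euler Z, remaps it to a clamped weight, and blends deform bones toward authored local offsets, rotations and scales.
# Elbows and knees layer two bend ranges.
# Shoulders and hips use absolute angles with different ranges per side, and only the right shoulder drives an extra delt_fr_r bone.
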ControlRigBlueprint /BRCosmetics/Characters/Player/Male/Large/Bodies/M_LRG_Clash_V/Meshes/M_LRG_Clash_V_CtrlRig
graphs [
 RigVMFunctionLibrary {links [], nodes []}
 RigVMModel {nodes [
   n1 = &RigUnit_BeginExecution()
   n2 = /Script/ControlRig.RigUnit_GetInitialBoneTransform(Bone=deform_elbow_bk_l, Space=LocalSpace) {is=struct RigVMUnitNode}
   n3 = &RigVMFunction_MathQuaternionToEuler(RotationOrder=ZYX, Value=@n2.Transform.Rotation)
   RigUnit_GetRelativeBoneTransform = &RigUnit_GetRelativeBoneTransform(Bone=lowerarm_l, Space=upperarm_l)
   n4 = &RigVMFunction_MathQuaternionToEuler(RotationOrder=ZYX, Value=@RigUnit_GetRelativeBoneTransform.Transform.Rotation)
   n5 = &RigVMFunction_MathFloatRemap(SourceMaximum=92, SourceMinimum=28, TargetMinimum=0, bClamp=true, Value=@n4.Result.Z)
   /Script/ControlRig.RigUnit_ModifyBoneTransforms(
    BoneToModify=((Bone="deform_elbow_bk_l",Transform=(Rotation=(X=0,Y=0,Z=0,W=1),Translation=(X=-2.858,Y=-2.932,Z=-0.666),Scale3D=(X=1,Y=1,Z=1))))
    Mode=OverrideLocal
    WeightMaximum=1
    BoneToModify.0.Transform.Rotation=&RigVMFunction_MathQuaternionFromEuler(Euler=/Script/RigVM.RigVMFunction_MathVectorAdd(B=(X=0,Y=0,Z=0), A=@n3.Result) {is=struct RigVMUnitNode, template="Add::Execute(in A,in B,out Result)"}.Result).Result
    BoneToModify.0.Transform.Translation=&RigVMFunction_MathVectorAdd(
     B=/Script/RigVM.RigVMFunction_MathVectorScale(Factor=1, Value=@n2.Transform.Translation) {is=struct RigVMUnitNode, template="Scale::Execute(in Value,in Factor,out Result)"}.Result
     A=&RigVMFunction_MathVectorScale(Factor=0.7, Value=(X=-8,Y=-5,Z=0)) {template="Scale::Execute(in Value,in Factor,out Result)"}.Result
    ) {template="Add::Execute(in A,in B,out Result)"}.Result
    ExecutePin=@n1.ExecutePin
    Weight=@n5.Result
   ) {is=struct RigVMUnitNode}
   n6 = &RigVMFunction_MathQuaternionToEuler(RotationOrder=ZYX, Value=&RigUnit_GetInitialBoneTransform(Bone=deform_elbow_fr_l, Space=LocalSpace).Transform.Rotation)
   > &RigUnit_ModifyBoneTransforms(
    BoneToModify=((Bone="deform_elbow_fr_l",Transform=(Rotation=(X=0,Y=0,Z=0,W=1),Translation=(X=24.635,Y=2.261,Z=-0.637),Scale3D=(X=1,Y=1,Z=1.125))))
    Mode=OverrideLocal
    WeightMaximum=1
    WeightMinimum=0
    BoneToModify.0.Transform.Rotation=&RigVMFunction_MathQuaternionFromEuler(Euler=&RigVMFunction_MathVectorAdd(B=(X=0,Y=0,Z=0), A=@n6.Result) {template="Add::Execute(in A,in B,out Result)"}.Result).Result
    Weight=@n5.Result
   )
   n7 = &RigVMFunction_MathFloatRemap(SourceMaximum=148, SourceMinimum=91, TargetMinimum=0, bClamp=true, Value=@n4.Result.Z)
   > &RigUnit_ModifyBoneTransforms(
    BoneToModify=((Bone="deform_elbow_bk_l",Transform=(Rotation=(X=0,Y=0,Z=0,W=1),Translation=(X=-3.59,Y=-2.934,Z=-0.716),Scale3D=(X=1,Y=1,Z=1))))
    Mode=OverrideLocal
    WeightMaximum=1
    Weight=@n7.Result
    BoneToModify.0.Transform.Rotation=&RigVMFunction_MathQuaternionFromEuler(Euler=&RigVMFunction_MathVectorAdd(B=(X=0,Y=0,Z=-13.6), A=@n3.Result) {template="Add::Execute(in A,in B,out Result)"}.Result).Result
   )
   > &RigUnit_ModifyBoneTransforms(
    BoneToModify=((Bone="deform_elbow_fr_l",Transform=(Rotation=(X=0,Y=0,Z=0,W=1),Translation=(X=15.651,Y=1.255,Z=-0.922),Scale3D=(X=1,Y=1,Z=1.337))))
    Mode=OverrideLocal
    WeightMaximum=1
    WeightMinimum=0
    Weight=@n7.Result
    BoneToModify.0.Transform.Rotation=&RigVMFunction_MathQuaternionFromEuler(Euler=&RigVMFunction_MathVectorAdd(B=(X=0,Y=0,Z=28), A=@n6.Result) {template="Add::Execute(in A,in B,out Result)"}.Result).Result
   )
   n8 = &RigVMFunction_MathQuaternionToEuler(RotationOrder=ZYX, Value=&RigUnit_GetRelativeBoneTransform(Bone=lowerarm_r, Space=upperarm_r).Transform.Rotation)
   n9 = &RigVMFunction_MathFloatRemap(SourceMaximum=92, SourceMinimum=28, TargetMinimum=0, bClamp=true, Value=@n8.Result.Z)
   n10 = &RigUnit_GetInitialBoneTransform(Bone=deform_elbow_bk_r, Space=LocalSpace)
   n11 = &RigVMFunction_MathQuaternionToEuler(RotationOrder=ZYX, Value=@n10.Transform.Rotation)
   > &RigUnit_ModifyBoneTransforms(
    BoneToModify=((Bone="deform_elbow_bk_r",Transform=(Rotation=(X=0,Y=0,Z=0,W=1),Translation=(X=2.858,Y=2.932,Z=0.666),Scale3D=(X=1,Y=1,Z=1))))
    Mode=OverrideLocal
    WeightMaximum=1
    Weight=@n9.Result
    BoneToModify.0.Transform.Rotation=&RigVMFunction_MathQuaternionFromEuler(Euler=&RigVMFunction_MathVectorAdd(B=(X=0,Y=0,Z=0), A=@n11.Result) {template="Add::Execute(in A,in B,out Result)"}.Result).Result
    BoneToModify.0.Transform.Translation=&RigVMFunction_MathVectorAdd(
     A=&RigVMFunction_MathVectorScale(Factor=1, Value=@n10.Transform.Translation) {template="Scale::Execute(in Value,in Factor,out Result)"}.Result
     B=&RigVMFunction_MathVectorScale(Factor=0.7, Value=(X=8,Y=5,Z=0)) {template="Scale::Execute(in Value,in Factor,out Result)"}.Result
    ) {template="Add::Execute(in A,in B,out Result)"}.Result
   )
   n12 = &RigVMFunction_MathQuaternionToEuler(RotationOrder=ZYX, Value=&RigUnit_GetInitialBoneTransform(Bone=deform_elbow_fr_r, Space=LocalSpace).Transform.Rotation)
   > &RigUnit_ModifyBoneTransforms(
    BoneToModify=((Bone="deform_elbow_fr_r",Transform=(Rotation=(X=0,Y=0,Z=0,W=1),Translation=(X=-24.635,Y=-2.261,Z=0.637),Scale3D=(X=1,Y=1,Z=1.125))))
    Mode=OverrideLocal
    WeightMaximum=1
    WeightMinimum=0
    Weight=@n9.Result
    BoneToModify.0.Transform.Rotation=&RigVMFunction_MathQuaternionFromEuler(Euler=&RigVMFunction_MathVectorAdd(B=(X=0,Y=0,Z=0), A=@n12.Result) {template="Add::Execute(in A,in B,out Result)"}.Result).Result
   )
   n13 = &RigVMFunction_MathFloatRemap(SourceMaximum=148, SourceMinimum=91, TargetMinimum=0, bClamp=true, Value=@n8.Result.Z)
   > &RigUnit_ModifyBoneTransforms(
    BoneToModify=((Bone="deform_elbow_bk_r",Transform=(Rotation=(X=0,Y=0,Z=0,W=1),Translation=(X=3.59,Y=2.934,Z=0.716),Scale3D=(X=1,Y=1,Z=1))))
    Mode=OverrideLocal
    WeightMaximum=1
    Weight=@n13.Result
    BoneToModify.0.Transform.Rotation=&RigVMFunction_MathQuaternionFromEuler(Euler=&RigVMFunction_MathVectorAdd(B=(X=0,Y=0,Z=-13.6), A=@n11.Result) {template="Add::Execute(in A,in B,out Result)"}.Result).Result
   )
   > &RigUnit_ModifyBoneTransforms(
    BoneToModify=((Bone="deform_elbow_fr_r",Transform=(Rotation=(X=0,Y=0,Z=0,W=1),Translation=(X=-15.651,Y=-1.255,Z=0.922),Scale3D=(X=1,Y=1,Z=1.337))))
    Mode=OverrideLocal
    WeightMaximum=1
    WeightMinimum=0
    Weight=@n13.Result
    BoneToModify.0.Transform.Rotation=&RigVMFunction_MathQuaternionFromEuler(Euler=&RigVMFunction_MathVectorAdd(B=(X=0,Y=0,Z=28), A=@n12.Result) {template="Add::Execute(in A,in B,out Result)"}.Result).Result
   )
   n14 = &RigVMFunction_MathQuaternionToEuler(RotationOrder=ZYX, Value=&RigUnit_GetRelativeBoneTransform(Bone=calf_l, Space=thigh_l).Transform.Rotation)
   n15 = &RigVMFunction_MathFloatRemap(SourceMaximum=90, SourceMinimum=4, TargetMinimum=0, bClamp=true, Value=@n14.Result.Z)
   n16 = &RigVMFunction_MathQuaternionToEuler(RotationOrder=ZYX, Value=&RigUnit_GetInitialBoneTransform(Bone=deform_knee_bk_01_l, Space=LocalSpace).Transform.Rotation)
   > &RigUnit_ModifyBoneTransforms(
    BoneToModify=((Bone="deform_knee_bk_01_l",Transform=(Rotation=(X=0,Y=0,Z=0,W=1),Translation=(X=-10,Y=0,Z=0),Scale3D=(X=1,Y=1.5,Z=1.5))))
    Mode=OverrideLocal
    WeightMaximum=1
    Weight=@n15.Result
    BoneToModify.0.Transform.Rotation=&RigVMFunction_MathQuaternionFromEuler(Euler=&RigVMFunction_MathVectorAdd(B=(X=0,Y=0,Z=0), A=@n16.Result) {template="Add::Execute(in A,in B,out Result)"}.Result).Result
   )
   n17 = &RigVMFunction_MathQuaternionToEuler(RotationOrder=ZYX, Value=&RigUnit_GetInitialBoneTransform(Bone=deform_knee_fr_01_l, Space=LocalSpace).Transform.Rotation)
   > &RigUnit_ModifyBoneTransforms(
    BoneToModify=((Bone="deform_knee_fr_01_l",Transform=(Rotation=(X=0,Y=0,Z=0,W=1),Translation=(X=1.701,Y=1.197,Z=0),Scale3D=(X=1,Y=1,Z=1))))
    Mode=OverrideLocal
    WeightMaximum=1
    WeightMinimum=0
    Weight=@n15.Result
    BoneToModify.0.Transform.Rotation=&RigVMFunction_MathQuaternionFromEuler(Euler=&RigVMFunction_MathVectorAdd(B=(X=0,Y=0,Z=0), A=@n17.Result) {template="Add::Execute(in A,in B,out Result)"}.Result).Result
   )
   n18 = &RigVMFunction_MathFloatRemap(SourceMaximum=130, SourceMinimum=89, TargetMinimum=0, bClamp=true, Value=@n14.Result.Z)
   > &RigUnit_ModifyBoneTransforms(
    BoneToModify=((Bone="deform_knee_bk_01_l",Transform=(Rotation=(X=0,Y=0,Z=0,W=1),Translation=(X=-10.633,Y=-3.561,Z=0),Scale3D=(X=1,Y=1,Z=1.777))))
    Mode=OverrideLocal
    WeightMaximum=1
    WeightMinimum=0
    Weight=@n18.Result
    BoneToModify.0.Transform.Rotation=&RigVMFunction_MathQuaternionFromEuler(Euler=&RigVMFunction_MathVectorAdd(B=(X=0,Y=0,Z=0), A=@n16.Result) {template="Add::Execute(in A,in B,out Result)"}.Result).Result
   )
   > &RigUnit_ModifyBoneTransforms(
    BoneToModify=((Bone="deform_knee_fr_01_l",Transform=(Rotation=(X=0,Y=0,Z=0,W=1),Translation=(X=1.791,Y=1.35,Z=0),Scale3D=(X=1.913,Y=1,Z=1))))
    Mode=OverrideLocal
    WeightMaximum=1
    WeightMinimum=0
    Weight=@n18.Result
    BoneToModify.0.Transform.Rotation=&RigVMFunction_MathQuaternionFromEuler(Euler=&RigVMFunction_MathVectorAdd(B=(X=0,Y=0,Z=0), A=@n17.Result) {template="Add::Execute(in A,in B,out Result)"}.Result).Result
   )
   n19 = &RigVMFunction_MathQuaternionToEuler(RotationOrder=ZYX, Value=&RigUnit_GetRelativeBoneTransform(Bone=calf_r, Space=thigh_r).Transform.Rotation)
   n20 = &RigVMFunction_MathFloatRemap(SourceMaximum=90, SourceMinimum=4, TargetMinimum=0, bClamp=true, Value=@n19.Result.Z)
   n21 = &RigVMFunction_MathQuaternionToEuler(RotationOrder=ZYX, Value=&RigUnit_GetInitialBoneTransform(Bone=deform_knee_bk_01_r, Space=LocalSpace).Transform.Rotation)
   > &RigUnit_ModifyBoneTransforms(
    BoneToModify=((Bone="deform_knee_bk_01_r",Transform=(Rotation=(X=0,Y=0,Z=0,W=1),Translation=(X=10,Y=0,Z=0),Scale3D=(X=1,Y=1.5,Z=1.5))))
    Mode=OverrideLocal
    WeightMaximum=1
    Weight=@n20.Result
    BoneToModify.0.Transform.Rotation=&RigVMFunction_MathQuaternionFromEuler(Euler=&RigVMFunction_MathVectorAdd(B=(X=0,Y=0,Z=0), A=@n21.Result) {template="Add::Execute(in A,in B,out Result)"}.Result).Result
   )
   n22 = &RigVMFunction_MathQuaternionToEuler(RotationOrder=ZYX, Value=&RigUnit_GetInitialBoneTransform(Bone=deform_knee_fr_01_r, Space=LocalSpace).Transform.Rotation)
   > &RigUnit_ModifyBoneTransforms(
    BoneToModify=((Bone="deform_knee_fr_01_r",Transform=(Rotation=(X=0,Y=0,Z=0,W=1),Translation=(X=-1.701,Y=-1.197,Z=0),Scale3D=(X=1,Y=1,Z=1))))
    Mode=OverrideLocal
    WeightMaximum=1
    WeightMinimum=0
    Weight=@n20.Result
    BoneToModify.0.Transform.Rotation=&RigVMFunction_MathQuaternionFromEuler(Euler=&RigVMFunction_MathVectorAdd(B=(X=0,Y=0,Z=0), A=@n22.Result) {template="Add::Execute(in A,in B,out Result)"}.Result).Result
   )
   n23 = &RigVMFunction_MathFloatRemap(SourceMaximum=130, SourceMinimum=89, TargetMinimum=0, bClamp=true, Value=@n19.Result.Z)
   > &RigUnit_ModifyBoneTransforms(
    BoneToModify=((Bone="deform_knee_bk_01_r",Transform=(Rotation=(X=0,Y=0,Z=0,W=1),Translation=(X=10.633,Y=3.561,Z=0),Scale3D=(X=1,Y=1,Z=1.777))))
    Mode=OverrideLocal
    WeightMaximum=1
    WeightMinimum=0
    Weight=@n23.Result
    BoneToModify.0.Transform.Rotation=&RigVMFunction_MathQuaternionFromEuler(Euler=&RigVMFunction_MathVectorAdd(B=(X=0,Y=0,Z=0), A=@n21.Result) {template="Add::Execute(in A,in B,out Result)"}.Result).Result
   )
   > &RigUnit_ModifyBoneTransforms(
    BoneToModify=((Bone="deform_knee_fr_01_r",Transform=(Rotation=(X=0,Y=0,Z=0,W=1),Translation=(X=-1.791,Y=-1.35,Z=0),Scale3D=(X=1.913,Y=1,Z=1))))
    Mode=OverrideLocal
    WeightMaximum=1
    WeightMinimum=0
    Weight=@n23.Result
    BoneToModify.0.Transform.Rotation=&RigVMFunction_MathQuaternionFromEuler(Euler=&RigVMFunction_MathVectorAdd(B=(X=0,Y=0,Z=0), A=@n22.Result) {template="Add::Execute(in A,in B,out Result)"}.Result).Result
   )
   n24 = &RigUnit_GetRelativeBoneTransform(Bone=upperarm_l, Space=spine_05)
   n25 = &RigVMFunction_MathFloatRemap(
    SourceMaximum=50
    SourceMinimum=170
    TargetMinimum=0
    bClamp=true
    Value=/Script/RigVM.RigVMFunction_MathFloatAbs(Value=&RigVMFunction_MathQuaternionToEuler(RotationOrder=ZYX, Value=@n24.Transform.Rotation).Result.Z) {is=struct RigVMUnitNode, template="Absolute::Execute(in Value,out Result)"}.Result
   )
   n26 = &RigVMFunction_MathQuaternionToEuler(RotationOrder=ZYX, Value=&RigUnit_GetInitialBoneTransform(Bone=deform_pectoral_fr_l, Space=LocalSpace).Transform.Rotation)
   > &RigUnit_ModifyBoneTransforms(
    BoneToModify=((Bone="deform_pectoral_fr_l",Transform=(Rotation=(X=0,Y=0,Z=0,W=1),Translation=(X=6.262,Y=5.9,Z=4.98),Scale3D=(X=1,Y=1,Z=1))))
    Mode=OverrideLocal
    WeightMaximum=1
    Weight=@n25.Result
    BoneToModify.0.Transform.Rotation=&RigVMFunction_MathQuaternionFromEuler(Euler=&RigVMFunction_MathVectorAdd(B=(X=0,Y=0,Z=0), A=@n26.Result) {template="Add::Execute(in A,in B,out Result)"}.Result).Result
   )
   n27 = &RigVMFunction_MathQuaternionToEuler(RotationOrder=ZYX, Value=&RigUnit_GetInitialBoneTransform(Bone=deform_deltoid_fr_l, Space=LocalSpace).Transform.Rotation)
   > &RigUnit_ModifyBoneTransforms(
    BoneToModify=((Bone="deform_deltoid_fr_l",Transform=(Rotation=(X=0,Y=0,Z=0,W=1),Translation=(X=5.008,Y=3.542,Z=0),Scale3D=(X=1,Y=1,Z=1))))
    Mode=OverrideLocal
    WeightMaximum=1
    WeightMinimum=0
    Weight=@n25.Result
    BoneToModify.0.Transform.Rotation=&RigVMFunction_MathQuaternionFromEuler(Euler=&RigVMFunction_MathVectorAdd(B=(X=0,Y=0,Z=-51), A=@n27.Result) {template="Add::Execute(in A,in B,out Result)"}.Result).Result
   )
   n28 = &RigUnit_GetRelativeBoneTransform(Bone=upperarm_r, Space=spine_05)
   n29 = &RigVMFunction_MathFloatRemap(
    SourceMaximum=123
    SourceMinimum=30
    TargetMinimum=0
    bClamp=true
    Value=&RigVMFunction_MathFloatAbs(Value=&RigVMFunction_MathQuaternionToEuler(RotationOrder=ZYX, Value=@n28.Transform.Rotation).Result.Z) {template="Absolute::Execute(in Value,out Result)"}.Result
   )
   RigUnit_GetInitialBoneTransform = &RigUnit_GetInitialBoneTransform(Bone=deform_pectoral_fr_r, Space=LocalSpace)
   n30 = &RigVMFunction_MathQuaternionToEuler(RotationOrder=ZYX, Value=@RigUnit_GetInitialBoneTransform.Transform.Rotation)
   > &RigUnit_ModifyBoneTransforms(
    BoneToModify=((Bone="deform_pectoral_fr_r",Transform=(Rotation=(X=0,Y=0,Z=0,W=1),Translation=(X=-6.262,Y=-5.9,Z=-4.98),Scale3D=(X=1,Y=1,Z=1))))
    Mode=OverrideLocal
    WeightMaximum=1
    Weight=@n29.Result
    BoneToModify.0.Transform.Rotation=&RigVMFunction_MathQuaternionFromEuler(Euler=&RigVMFunction_MathVectorAdd(B=(X=0,Y=0,Z=0), A=@n30.Result) {template="Add::Execute(in A,in B,out Result)"}.Result).Result
   )
   n31 = &RigVMFunction_MathQuaternionToEuler(RotationOrder=ZYX, Value=&RigUnit_GetInitialBoneTransform(Bone=deform_deltoid_fr_r, Space=LocalSpace).Transform.Rotation)
   > &RigUnit_ModifyBoneTransforms(
    BoneToModify=((Bone="deform_deltoid_fr_r",Transform=(Rotation=(X=0,Y=0,Z=0,W=1),Translation=(X=-15,Y=0,Z=15),Scale3D=(X=2,Y=2,Z=2))))
    Mode=OverrideLocal
    WeightMaximum=1
    WeightMinimum=0
    Weight=@n29.Result
    BoneToModify.0.Transform.Rotation=&RigVMFunction_MathQuaternionFromEuler(Euler=&RigVMFunction_MathVectorAdd(B=(X=0,Y=0,Z=-51), A=@n31.Result) {template="Add::Execute(in A,in B,out Result)"}.Result).Result
   )
   n32 = &RigVMFunction_MathQuaternionToEuler(RotationOrder=ZYX, Value=&RigUnit_GetInitialBoneTransform(Bone=delt_fr_r, Space=LocalSpace).Transform.Rotation)
   > &RigUnit_ModifyBoneTransforms(
    BoneToModify=((Bone="delt_fr_r",Transform=(Rotation=(X=0,Y=0,Z=0,W=1),Translation=(X=0,Y=0,Z=-5),Scale3D=(X=1,Y=1,Z=1))))
    Mode=OverrideLocal
    WeightMaximum=1
    WeightMinimum=0
    Weight=@n29.Result
    BoneToModify.0.Transform.Rotation=&RigVMFunction_MathQuaternionFromEuler(Euler=&RigVMFunction_MathVectorAdd(B=(X=0,Y=0,Z=0), A=@n32.Result) {template="Add::Execute(in A,in B,out Result)"}.Result).Result
   )
   n33 = &RigUnit_GetRelativeBoneTransform(Bone=thigh_l, Space=pelvis)
   n34 = &RigVMFunction_MathFloatRemap(
    SourceMaximum=100
    SourceMinimum=2
    TargetMinimum=0
    bClamp=true
    Value=&RigVMFunction_MathFloatAbs(Value=&RigVMFunction_MathQuaternionToEuler(RotationOrder=ZYX, Value=@n33.Transform.Rotation).Result.Z) {template="Absolute::Execute(in Value,out Result)"}.Result
   )
   n35 = &RigVMFunction_MathQuaternionToEuler(RotationOrder=ZYX, Value=&RigUnit_GetInitialBoneTransform(Bone=deform_thigh_fr_l, Space=LocalSpace).Transform.Rotation)
   > &RigUnit_ModifyBoneTransforms(
    BoneToModify=((Bone="deform_thigh_fr_l",Transform=(Rotation=(X=0,Y=0,Z=0,W=1),Translation=(X=-8.207,Y=8.962,Z=2.253),Scale3D=(X=1,Y=1,Z=1.175))))
    Mode=OverrideLocal
    WeightMaximum=1
    Weight=@n34.Result
    BoneToModify.0.Transform.Rotation=&RigVMFunction_MathQuaternionFromEuler(Euler=&RigVMFunction_MathVectorAdd(B=(X=0,Y=-14.949,Z=0), A=@n35.Result) {template="Add::Execute(in A,in B,out Result)"}.Result).Result
   )
   n36 = &RigVMFunction_MathQuaternionToEuler(RotationOrder=ZYX, Value=&RigUnit_GetInitialBoneTransform(Bone=deform_thigh_bk_l, Space=LocalSpace).Transform.Rotation)
   > &RigUnit_ModifyBoneTransforms(
    BoneToModify=((Bone="deform_thigh_bk_l",Transform=(Rotation=(X=0,Y=0,Z=0,W=1),Translation=(X=8.328,Y=-1.401,Z=0.029),Scale3D=(X=1,Y=1,Z=1.803))))
    Mode=OverrideLocal
    WeightMaximum=1
    WeightMinimum=0
    Weight=@n34.Result
    BoneToModify.0.Transform.Rotation=&RigVMFunction_MathQuaternionFromEuler(Euler=&RigVMFunction_MathVectorAdd(B=(X=0,Y=-37.217,Z=0), A=@n36.Result) {template="Add::Execute(in A,in B,out Result)"}.Result).Result
   )
   n37 = &RigUnit_GetRelativeBoneTransform(Bone=thigh_r, Space=pelvis)
   n38 = &RigVMFunction_MathFloatRemap(
    SourceMaximum=77
    SourceMinimum=177
    TargetMinimum=0
    bClamp=true
    Value=&RigVMFunction_MathFloatAbs(Value=&RigVMFunction_MathQuaternionToEuler(RotationOrder=ZYX, Value=@n37.Transform.Rotation).Result.Z) {template="Absolute::Execute(in Value,out Result)"}.Result
   )
   n39 = &RigVMFunction_MathQuaternionToEuler(RotationOrder=ZYX, Value=&RigUnit_GetInitialBoneTransform(Bone=deform_thigh_fr_r, Space=LocalSpace).Transform.Rotation)
   > &RigUnit_ModifyBoneTransforms(
    BoneToModify=((Bone="deform_thigh_fr_r",Transform=(Rotation=(X=0,Y=0,Z=0,W=1),Translation=(X=8.207,Y=-8.962,Z=-2.253),Scale3D=(X=1,Y=1,Z=1.175))))
    Mode=OverrideLocal
    WeightMaximum=1
    Weight=@n38.Result
    BoneToModify.0.Transform.Rotation=&RigVMFunction_MathQuaternionFromEuler(Euler=&RigVMFunction_MathVectorAdd(B=(X=0,Y=-14.949,Z=0), A=@n39.Result) {template="Add::Execute(in A,in B,out Result)"}.Result).Result
   )
   n40 = &RigVMFunction_MathQuaternionToEuler(RotationOrder=ZYX, Value=&RigUnit_GetInitialBoneTransform(Bone=deform_thigh_bk_r, Space=LocalSpace).Transform.Rotation)
   > &RigUnit_ModifyBoneTransforms(
    BoneToModify=((Bone="deform_thigh_bk_r",Transform=(Rotation=(X=0,Y=0,Z=0,W=1),Translation=(X=-8.328,Y=1.401,Z=-0.029),Scale3D=(X=1,Y=1,Z=1.803))))
    Mode=OverrideLocal
    WeightMaximum=1
    WeightMinimum=0
    Weight=@n38.Result
    BoneToModify.0.Transform.Rotation=&RigVMFunction_MathQuaternionFromEuler(Euler=&RigVMFunction_MathVectorAdd(B=(X=0,Y=-37.217,Z=0), A=@n40.Result) {template="Add::Execute(in A,in B,out Result)"}.Result).Result
   )
   > /Script/ControlRig.RigUnit_DebugHierarchy(
    Color=(R=1,G=1,B=1,A=1)
    DebugDrawSettings=(DepthPriority=SDPG_Foreground,Lifetime=-1)
    Items=()
    Scale=3
    Thickness=0.1
    WorldOffset=(Rotation=(X=0,Y=0,Z=0,W=1),Translation=(X=0,Y=0,Z=0),Scale3D=(X=1,Y=1,Z=1))
    bEnabled=True
   ) {is=struct RigVMUnitNode}
  ]}
]
hierarchy [
 root {loc=<0 0 0>, rot=<0 0 0 1>}
 attach {parent=root, loc=<0 0 0>, rot=<0 0 0 1>}
 pelvis {parent=root, loc=<0 -0.7321 92.4218>, rot=<0.0127 0.707 -0.0127 -0.707>}
 spine_01 {parent=pelvis, loc=<5.6002 -0.5128 0>, rot=<0 0 0.034 -0.9994>}
 spine_02 {parent=spine_01, loc=<8.0753 0.0139 0>, rot=<0 0 0.0238 -0.9997>}
 spine_03 {parent=spine_02, loc=<7.0393 0 0>, rot=<0 0 0.0149 -0.9999>}
 spine_04 {parent=spine_03, loc=<8.1023 0 0>, rot=<0 0 -0.0132 -0.9999>}
 spine_05 {parent=spine_04, loc=<10.0146 0 0>, rot=<0 0 -0.0015 -1>}
 clavicle_l {parent=spine_05, loc=<2.5855 -0.2866 -4.3777>, rot=<0.0465 -0.7289 -0.009 -0.6829>}
 upperarm_l {parent=clavicle_l, loc=<18.2264 0.1342 -0.001>, rot=<0.0253 -0.2693 0.0013 -0.9627>}
 lowerarm_l {parent=upperarm_l, loc=<29.0071 0.1686 -1.239>, rot=<0.0391 -0.0099 -0.2453 -0.9686>}
 hand_l {parent=lowerarm_l, loc=<25.8607 0.222 1.0274>, rot=<-0.7193 -0.0988 0.1193 -0.6773>}
 index_metacarpal_l {parent=hand_l, loc=<1.6535 0.0183 -2.96>, rot=<0.0158 -0.0124 -0.0173 -0.9996>}
 index_01_l {parent=index_metacarpal_l, loc=<8.3008 0.0357 -0.0524>, rot=<-0.0044 0.0485 0.0924 -0.9945>}
 index_02_l {parent=index_01_l, loc=<3.3587 -0.3003 0.0452>, rot=<-0.002 0.0027 0.3225 -0.9466>}
 index_03_l {parent=index_02_l, loc=<2.5578 -0.123 -0.0252>, rot=<-0.002 0.0067 0.2558 -0.9667>}
 middle_metacarpal_l {parent=hand_l, loc=<2.1856 -0.1725 -0.5984>, rot=<0.006 0.0168 -0.0329 -0.9993>}
 middle_01_l {parent=middle_metacarpal_l, loc=<7.3275 -0.2448 0.0129>, rot=<-0.0208 0.0042 0.0858 -0.9961>}
 middle_02_l {parent=middle_01_l, loc=<4.0123 0.1402 0.1175>, rot=<-0.0218 0.0258 0.3512 -0.9357>}
 middle_03_l {parent=middle_02_l, loc=<2.6077 0.0534 0.1373>, rot=<0.0082 0.0188 0.1708 -0.9851>}
 pinky_metacarpal_l {parent=hand_l, loc=<1.7616 -0.9753 3.3814>, rot=<-0.0397 0.0814 -0.069 -0.9935>}
 pinky_01_l {parent=pinky_metacarpal_l, loc=<7.1709 -0.4831 -0.1094>, rot=<-0.0654 -0.0035 0.2242 -0.9723>}
 pinky_02_l {parent=pinky_01_l, loc=<2.8274 0.0893 -0.1043>, rot=<0.0037 0.0011 0.2854 -0.9584>}
 pinky_03_l {parent=pinky_02_l, loc=<2.2046 0.0307 -0.0078>, rot=<-0.0002 0.0098 0.1445 -0.9895>}
 ring_metacarpal_l {parent=hand_l, loc=<2.2074 -0.3625 1.3638>, rot=<-0.0284 0.0722 -0.0458 -0.9959>}
 ring_01_l {parent=ring_metacarpal_l, loc=<7.0295 -0.6197 -0.1446>, rot=<-0.0489 -0.0252 0.1892 -0.9804>}
 ring_02_l {parent=ring_01_l, loc=<3.6174 0.4647 -0.1027>, rot=<0 0 0.2806 -0.9598>}
 ring_03_l {parent=ring_02_l, loc=<2.4152 0.114 0.0199>, rot=<0 0 0.1189 -0.9929>}
 thumb_01_l {parent=hand_l, loc=<1.3118 -0.3191 -2.459>, rot=<0.6546 -0.4042 0.0664 -0.6354>}
 thumb_02_l {parent=thumb_01_l, loc=<5.9336 -0.0448 0.0401>, rot=<-0.0037 0.0213 0.298 -0.9543>}
 thumb_03_l {parent=thumb_02_l, loc=<2.2223 0.3682 -0.0907>, rot=<0.0118 -0.013 0.0066 -0.9998>}
 weapon_l {parent=hand_l, loc=<-0.1974 -3.4111 0.3811>, rot=<0 0 0 -1>}
 lowerarm_twist_01_l {parent=lowerarm_l, loc=<17.8549 0 0>, rot=<0 0 0 -1>}
 lowerarm_twist_02_l {parent=lowerarm_l, loc=<9.6199 0 0>, rot=<0 0 0 -1>}
 deform_elbow_bk_l {parent=lowerarm_twist_02_l, loc=<-10.9052 -2.9274 -0.5607>, rot=<-0.0285 0.0335 -0.0024 -0.999>}
 upperarm_twist_01_l {parent=upperarm_l, loc=<7.9796 0.0843 -0.6195>, rot=<0 0 0 -1>}
 deform_deltoid_fr_l {parent=upperarm_twist_01_l, loc=<2.4511 0 0>, rot=<0 0 0 1>}
 upperarm_twist_02_l {parent=upperarm_l, loc=<19.5059 0.0843 -0.6195>, rot=<0 0 0 -1>}
 deform_elbow_fr_l {parent=upperarm_l, loc=<26.5048 0.1875 -0.5816>, rot=<0.0021 0.0156 -0.185 -0.9826>}
 tricep_l {parent=upperarm_l, loc=<8.4813 -13.6068 -2.7101>, rot=<0 0 1 0>}
 deform_pectoral_fr_l {parent=clavicle_l, loc=<9.3905 0 0>, rot=<0 0 0 1>}
 scap_l {parent=clavicle_l, loc=<13.5884 -10.4244 0.2351>, rot=<0 0 1 0>}
 clavicle_r {parent=spine_05, loc=<2.5856 -0.2865 4.3777>, rot=<0.7289 0.0465 -0.6829 0.009>}
 upperarm_r {parent=clavicle_r, loc=<-18.2264 -0.1342 0.001>, rot=<0.0253 -0.2693 0.0013 -0.9627>}
 lowerarm_r {parent=upperarm_r, loc=<-29.0072 -0.1686 1.239>, rot=<0.0391 -0.0099 -0.2453 -0.9686>}
 hand_r {parent=lowerarm_r, loc=<-25.8608 -0.222 -1.0274>, rot=<-0.7193 -0.0988 0.1193 -0.6773>}
 index_metacarpal_r {parent=hand_r, loc=<-1.6534 -0.0181 2.9601>, rot=<0.0158 -0.0124 -0.0173 -0.9996>}
 index_01_r {parent=index_metacarpal_r, loc=<-8.3008 -0.0357 0.0524>, rot=<-0.0044 0.0485 0.0924 -0.9945>}
 index_02_r {parent=index_01_r, loc=<-3.3589 0.3003 -0.0453>, rot=<-0.002 0.0027 0.3225 -0.9466>}
 index_03_r {parent=index_02_r, loc=<-2.5578 0.123 0.0252>, rot=<-0.002 0.0067 0.2558 -0.9667>}
 middle_metacarpal_r {parent=hand_r, loc=<-2.1857 0.1726 0.5984>, rot=<0.006 0.0168 -0.0329 -0.9993>}
 middle_01_r {parent=middle_metacarpal_r, loc=<-7.3273 0.2447 -0.0129>, rot=<-0.0208 0.0042 0.0858 -0.9961>}
 middle_02_r {parent=middle_01_r, loc=<-4.0124 -0.14 -0.1175>, rot=<-0.0218 0.0258 0.3512 -0.9357>}
 middle_03_r {parent=middle_02_r, loc=<-2.6077 -0.0533 -0.1373>, rot=<0.0082 0.0188 0.1708 -0.9851>}
 pinky_metacarpal_r {parent=hand_r, loc=<-1.7617 0.9753 -3.3814>, rot=<-0.0397 0.0814 -0.069 -0.9935>}
 pinky_01_r {parent=pinky_metacarpal_r, loc=<-7.1707 0.4832 0.1094>, rot=<-0.0654 -0.0035 0.2242 -0.9723>}
 pinky_02_r {parent=pinky_01_r, loc=<-2.8274 -0.0892 0.1043>, rot=<0.0037 0.0011 0.2854 -0.9584>}
 pinky_03_r {parent=pinky_02_r, loc=<-2.2045 -0.031 0.0078>, rot=<-0.0002 0.0098 0.1445 -0.9895>}
 ring_metacarpal_r {parent=hand_r, loc=<-2.2073 0.3627 -1.3638>, rot=<-0.0284 0.0722 -0.0458 -0.9959>}
 ring_01_r {parent=ring_metacarpal_r, loc=<-7.0297 0.6197 0.1446>, rot=<-0.0489 -0.0252 0.1892 -0.9804>}
 ring_02_r {parent=ring_01_r, loc=<-3.6171 -0.4648 0.1027>, rot=<0 0 0.2806 -0.9598>}
 ring_03_r {parent=ring_02_r, loc=<-2.4153 -0.1141 -0.0199>, rot=<0 0 0.1189 -0.9929>}
 thumb_01_r {parent=hand_r, loc=<-1.3119 0.3193 2.459>, rot=<0.6546 -0.4042 0.0664 -0.6354>}
 thumb_02_r {parent=thumb_01_r, loc=<-5.9335 0.0447 -0.0401>, rot=<-0.0037 0.0213 0.298 -0.9543>}
 thumb_03_r {parent=thumb_02_r, loc=<-2.2223 -0.3682 0.0907>, rot=<0.0118 -0.013 0.0066 -0.9998>}
 weapon_r {parent=hand_r, loc=<0.1975 3.4112 -0.3811>, rot=<0 0 0 1>}
 lowerarm_twist_01_r {parent=lowerarm_r, loc=<-17.855 0 0>, rot=<0 0 0 -1>}
 lowerarm_twist_02_r {parent=lowerarm_r, loc=<-9.6199 0 0>, rot=<0 0 0 -1>}
 deform_elbow_bk_r {parent=lowerarm_r, loc=<1.2852 2.9274 0.561>, rot=<-0.0285 0.0335 -0.0024 -0.999>}
 upperarm_twist_01_r {parent=upperarm_r, loc=<-7.9796 -0.0843 0.6195>, rot=<0 0 0 -1>}
 deform_deltoid_fr_r {parent=upperarm_twist_01_r, loc=<-2.4512 0 0>, rot=<0 0 0 1>}
 delt_fr_r {parent=upperarm_twist_01_r, loc=<0 0 -4>, rot=<0 0 0 -1>}
 upperarm_twist_02_r {parent=upperarm_r, loc=<-19.506 -0.0843 0.6195>, rot=<0 0 0 -1>}
 deform_elbow_fr_r {parent=upperarm_r, loc=<-26.5046 -0.1875 0.5813>, rot=<0.0021 0.0156 -0.185 -0.9826>}
 shldr_half_r {parent=upperarm_r, loc=<0 0 0>, rot=<0 0 0 1>}
 tricep_r {parent=upperarm_r, loc=<-8.4813 13.6069 2.7102>, rot=<0 0 0 -1>}
 deform_pectoral_fr_r {parent=clavicle_r, loc=<-9.397 0 0>, rot=<0 0 0 1>}
 scap_r {parent=clavicle_r, loc=<-13.5884 10.4244 -0.235>, rot=<0 0 0 1>}
 neck_01 {parent=spine_05, loc=<10.815 1.9196 0>, rot=<0 0 -0.1864 -0.9825>}
 neck_02 {parent=neck_01, loc=<4.7347 0.3045 0>, rot=<0 0 0.0363 -0.9993>}
 head {parent=neck_02, loc=<5.7632 0.5552 0>, rot=<0 0 0.115 -0.9934>}
 faceAttach {parent=head, loc=<0.6419 -1.7133 0>, rot=<0.0082 -0.7071 -0.0082 -0.7071>, scale=<0.89 0.89 0.89>}
 C_brow_mid {parent=faceAttach, loc=<-0.003 10.6892 9.8462>, rot=<-0.0074 0 0 -1>}
 L_brow_mid {parent=faceAttach, loc=<0.4634 2.506 5.1454>, rot=<-0.0074 -0.0007 0.0881 -0.9961>}
 L_brow_outer {parent=faceAttach, loc=<1.0344 1.1958 6.8681>, rot=<-0.0069 -0.0027 0.362 -0.9321>}
 L_cheek_inner {parent=faceAttach, loc=<6.8202 4.2792 6.4785>, rot=<0.0959 0.1451 0.7717 -0.6118>}
 C_lip_upper_mid {parent=faceAttach, loc=<0 11.6961 0.8154>, rot=<-0.0074 0 0 -1>}
 L_lip_upper_outer {parent=faceAttach, loc=<6.7851 3.3397 4.0827>, rot=<0.1395 0.1782 0.5804 -0.7822>}
 L_lip_corner {parent=faceAttach, loc=<7.4453 -1.7911 6.6635>, rot=<-0.0074 0 0 -1>}
 C_nose_bridge {parent=faceAttach, loc=<0.0054 11.9017 6.6516>, rot=<-0.0074 0 0 -1>}
 R_brow_mid {parent=faceAttach, loc=<-0.4634 2.506 5.1454>, rot=<-0.0074 0.0007 -0.0881 -0.9961>}
 R_brow_outer {parent=faceAttach, loc=<-1.0344 1.1958 6.8681>, rot=<-0.0069 0.0027 -0.362 -0.9321>}
 R_cheek_inner {parent=faceAttach, loc=<-6.8202 4.2792 6.4785>, rot=<0.123 -0.1315 -0.5852 -0.7906>}
 R_lip_corner {parent=faceAttach, loc=<-7.4453 -1.7911 6.6635>, rot=<-0.0074 0 0 -1>}
 C_jaw {parent=faceAttach, loc=<0 -1.3446 3.1231>, rot=<0 0 0 1>}
 L_lip_lower_outer {parent=C_jaw, loc=<5.8336 4.6413 -4.1837>, rot=<0.1992 0.267 0.5618 -0.7573>}
 C_lip_lower_mid {parent=C_jaw, loc=<0 10.0398 -9.6195>, rot=<-0.0074 0 0 -1>}
 R_lip_lower_outer {parent=C_jaw, loc=<-5.8336 4.6413 -4.1837>, rot=<0.2696 -0.2696 -0.5167 -0.7666>}
 teeth_lower {parent=C_jaw, loc=<0 7.3085 -7.7877>, rot=<0 0 0 1>}
 tongue {parent=C_jaw, loc=<0 5.1808 -4.4676>, rot=<-0.1568 0 0 -0.9876>}
 R_eye {parent=faceAttach, loc=<-2.974 5.1333 8.9369>, rot=<0 0 0.9998 -0.0187>}
 R_eye_lid_upper_mid {parent=R_eye, loc=<-0.0396 -1.1714 0.2631>, rot=<0.0523 -0.0009 0.0167 -0.9985>}
 R_eye_lid_lower_mid {parent=R_eye, loc=<-0.0085 -0.9425 -0.7051>, rot=<0.0329 -0.0005 0.0167 -0.9993>}
 L_eye {parent=faceAttach, loc=<2.9919 5.2168 8.9559>, rot=<0 0 0.9998 0.0187>}
 L_eye_lid_upper_mid {parent=L_eye, loc=<0.0251 -1.1623 0.2388>, rot=<0.0523 0 0 -0.9986>}
 L_eye_lid_lower_mid {parent=L_eye, loc=<0.0223 -0.923 -0.7488>, rot=<0.0329 0 0 -0.9995>}
 teeth_upper {parent=faceAttach, loc=<0.0369 5.9516 1.3719>, rot=<0 0 0 1>}
 R_lip_upper_outer {parent=faceAttach, loc=<-6.7851 3.3397 4.0827>, rot=<0.0806 -0.1233 -0.682 -0.7164>}
 thigh_l {parent=pelvis, loc=<-1.8716 -0.0109 -11.9936>, rot=<0.0012 0.0572 0.0155 -0.9982>}
 calf_l {parent=thigh_l, loc=<-41.9178 -0.2246 0.0926>, rot=<0.0006 -0.0161 -0.0391 -0.9991>}
 calf_twist_01_l {parent=calf_l, loc=<-32.9335 -0.0967 0.0397>, rot=<0.0237 0.0095 -0.0035 -0.9997>}
 calf_twist_02_l {parent=calf_l, loc=<-15.9268 -0.0967 0.0397>, rot=<0.0237 0.0095 -0.0035 -0.9997>}
 foot_l {parent=calf_l, loc=<-38.7057 1.2438 0.2453>, rot=<0.0975 -0.0474 0.0317 -0.9936>}
 ball_l {parent=foot_l, loc=<-6.1614 13.1469 0.6807>, rot=<0.0001 -0.0001 -0.7063 -0.7079>}
 deform_knee_bk_01_l {parent=calf_l, loc=<0.6022 -1.0607 0>, rot=<0 0 -0.0408 -0.9992>}
 deform_knee_fr_01_l {parent=calf_l, loc=<-0.2342 0 0>, rot=<0 0 0 1>}
 thigh_twist_01_l {parent=thigh_l, loc=<-21.6303 0 0>, rot=<0.0013 0 0 -1>}
 deform_thigh_fr_l {parent=thigh_l, loc=<-0.1219 0.228 -0.0604>, rot=<0 -0.0217 0 -0.9998>}
 deform_thigh_bk_l {parent=thigh_l, loc=<-4.6789 0.2063 0.1802>, rot=<0 -0.0217 0 -0.9998>}
 thigh_r {parent=pelvis, loc=<-1.8714 -0.0109 11.9936>, rot=<0.0572 -0.0012 0.9982 0.0155>}
 calf_r {parent=thigh_r, loc=<41.918 0.2246 -0.0926>, rot=<0.0006 -0.0161 -0.0391 -0.9991>}
 calf_twist_01_r {parent=calf_r, loc=<32.9336 0.0967 -0.0397>, rot=<0.0237 0.0095 -0.0035 -0.9997>}
 calf_twist_02_r {parent=calf_r, loc=<15.9269 0.0967 -0.0397>, rot=<0.0237 0.0095 -0.0035 -0.9997>}
 foot_r {parent=calf_r, loc=<38.7058 -1.2438 -0.2453>, rot=<0.0975 -0.0474 0.0317 -0.9936>}
 ball_r {parent=foot_r, loc=<6.1614 -13.1469 -0.6807>, rot=<0.0001 -0.0001 -0.7063 -0.7079>}
 deform_knee_fr_01_r {parent=calf_r, loc=<0.2341 0 0>, rot=<0 0 0 1>}
 deform_knee_bk_01_r {parent=calf_r, loc=<-0.6022 1.0607 0>, rot=<0 0 -0.0408 -0.9992>}
 thigh_twist_01_r {parent=thigh_r, loc=<21.6303 0 0>, rot=<0.0013 0 0 -1>}
 deform_thigh_bk_r {parent=thigh_r, loc=<4.6791 -0.2063 -0.1802>, rot=<0 -0.0217 0 -0.9998>}
 deform_thigh_fr_r {parent=thigh_r, loc=<0.122 -0.228 0.0603>, rot=<0 -0.0217 0 -0.9998>}
 ik_foot_root {parent=root, loc=<0 0 0>, rot=<0 0 0 1>}
 ik_foot_l {parent=ik_foot_root, loc=<19.6189 -3.2051 10.3339>, rot=<-0.0753 -0.7018 -0.0677 0.7051>}
 ik_foot_r {parent=ik_foot_root, loc=<-19.6189 -3.2051 10.3338>, rot=<-0.7051 -0.0677 0.7018 -0.0753>}
 ik_hand_root {parent=root, loc=<0 0 0>, rot=<0 0 0 1>}
 ik_hand_gun {parent=ik_hand_root, loc=<-64.4621 4.8638 101.1961>, rot=<-0.6084 -0.1793 -0.4613 0.6205>}
 ik_hand_l {parent=ik_hand_gun, loc=<65.7709 101.9259 43.6659>, rot=<-0.5102 -0.7906 -0.3387 0>}
 ik_hand_r {parent=ik_hand_gun, loc=<0 0 0>, rot=<0 0 0 -1>}
]
