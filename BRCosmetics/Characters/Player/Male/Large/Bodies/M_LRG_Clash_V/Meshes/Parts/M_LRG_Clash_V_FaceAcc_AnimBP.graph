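# Anim Blueprint for the Clash_V large male face accessory, driving its tongue from the head's jaw opening.
# Copies the attached parent's pose, constrains tongue_base to head, blends a Control Rig by Jaw Open, then scales tongue_base.
# Updates derive both values from C_jaw twist, switch scale range during non-frontend emotes, and set a material hide parameter for the Vertigo emote.
# The tongue Anim Dynamics chain is unconnected to the output.
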AnimBlueprint /BRCosmetics/Characters/Player/Male/Large/Bodies/M_LRG_Clash_V/Meshes/Parts/M_LRG_Clash_V_FaceAcc_AnimBP
target_skeleton=&/M_LRG_Clash_V_FaceAcc_Skeleton
graphs [
 AnimGraph {nodes [
   n1 = Anim Dynamics - Bone: tongue_2(Alpha=1, ExternalForce="0,0,0") {
    angularDampingOverride=1
    angularSpringConstant=1500
    bAngularSpring=true
    bGravityOverrideInSimSpace=true
    bOverrideAngularDamping=true
    bUseSphericalLimits=true
    gravityScale=0.35
    simulationSpace=RootRelative
    alphaBoolBlend {blendOption=Linear}
    alphaScaleBiasClamp {bInterpResult=true, scale=0.3}
    boundBone {boneName=tongue_2}
    physicsBodyDefinitions [{
      collisionType=CustomSphere
      sphereCollisionRadius=2
      boundBone {boneName=tongue_2}
      boxExtents=<10 20 20>
      constraintSetup {angularLimitsMax=<5 60 60>, angularLimitsMin=<-5 -60 -60>, angularTarget=<0 0 0>}
      localJointOffset=<10 0 0>
     }]
    sphericalLimits [{limitRadius=34.3949, drivingBone {boneName=spine_05}, sphereLocalOffset=<-7.2181 -16.881 0>}]
   }
   n2 = Anim Dynamics - Bone: tongue_3(ComponentPose=@n1, Alpha=1, ExternalForce="0,0,0") {
    angularDampingOverride=1
    angularSpringConstant=1500
    bAngularSpring=true
    bGravityOverrideInSimSpace=true
    bOverrideAngularDamping=true
    bUseSphericalLimits=true
    gravityScale=0.35
    simulationSpace=RootRelative
    alphaBoolBlend {blendOption=Linear}
    alphaScaleBiasClamp {bInterpResult=true, scale=0.7}
    boundBone {boneName=tongue_3}
    physicsBodyDefinitions [{
      collisionType=CustomSphere
      sphereCollisionRadius=2
      boundBone {boneName=tongue_3}
      boxExtents=<10 20 20>
      constraintSetup {angularLimitsMax=<5 60 60>, angularLimitsMin=<-5 -60 -60>, angularTarget=<0 0 0>}
      localJointOffset=<10 0 0>
     }]
    sphericalLimits [{limitRadius=34.3949, drivingBone {boneName=spine_05}, sphereLocalOffset=<-7.2181 -16.881 0>}]
   }
   n3 = Anim Dynamics - Bone: tongue_4(ComponentPose=@n2, Alpha=1, ExternalForce="0,0,0") {
    angularDampingOverride=1
    angularSpringConstant=1500
    bAngularSpring=true
    bGravityOverrideInSimSpace=true
    bOverrideAngularDamping=true
    bUseSphericalLimits=true
    gravityScale=0.35
    simulationSpace=RootRelative
    alphaBoolBlend {blendOption=Linear}
    alphaScaleBiasClamp {bInterpResult=true}
    boundBone {boneName=tongue_4}
    physicsBodyDefinitions [{
      collisionType=CustomSphere
      sphereCollisionRadius=2
      boundBone {boneName=tongue_4}
      boxExtents=<10 20 20>
      constraintSetup {angularLimitsMax=<5 60 60>, angularLimitsMin=<-5 -60 -60>, angularTarget=<0 0 0>}
      localJointOffset=<10 0 0>
     }]
    sphericalLimits [{limitRadius=34.3949, drivingBone {boneName=spine_05}, sphereLocalOffset=<-7.2181 -16.881 0>}]
   }
   n4 = Anim Dynamics - Bone: tongue_5(ComponentPose=@n3, Alpha=1, ExternalForce="0,0,0") {
    angularDampingOverride=1
    angularSpringConstant=1500
    bAngularSpring=true
    bGravityOverrideInSimSpace=true
    bOverrideAngularDamping=true
    bUseSphericalLimits=true
    gravityScale=0.35
    simulationSpace=RootRelative
    alphaBoolBlend {blendOption=Linear}
    alphaScaleBiasClamp {bInterpResult=true}
    boundBone {boneName=tongue_5}
    physicsBodyDefinitions [{
      collisionType=CustomSphere
      sphereCollisionRadius=2
      boundBone {boneName=tongue_5}
      boxExtents=<10 20 20>
      constraintSetup {angularLimitsMax=<5 60 60>, angularLimitsMin=<-5 -60 -60>, angularTarget=<0 0 0>}
      localJointOffset=<10 0 0>
     }]
    sphericalLimits [{limitRadius=34.3949, drivingBone {boneName=spine_05}, sphereLocalOffset=<-7.2181 -16.881 0>}]
   }
   Anim Dynamics - Bone: tongue_6(ComponentPose=@n4, Alpha=1, ExternalForce="0,0,0") {
    angularDampingOverride=1
    angularSpringConstant=1500
    bAngularSpring=true
    bGravityOverrideInSimSpace=true
    bOverrideAngularDamping=true
    bUseSphericalLimits=true
    gravityScale=0.35
    simulationSpace=RootRelative
    alphaBoolBlend {blendOption=Linear}
    alphaScaleBiasClamp {bInterpResult=true}
    boundBone {boneName=tongue_6}
    physicsBodyDefinitions [{
      collisionType=CustomSphere
      sphereCollisionRadius=2
      boundBone {boneName=tongue_6}
      boxExtents=<10 20 20>
      constraintSetup {angularLimitsMax=<5 60 60>, angularLimitsMin=<-5 -60 -60>, angularTarget=<0 0 0>}
      localJointOffset=<10 0 0>
     }]
    sphericalLimits [{limitRadius=34.3949, drivingBone {boneName=spine_05}, sphereLocalOffset=<-7.2181 -16.881 0>}]
   }
   Local To Component()
   n5 = "Constraint - tongue_base (P)"(ComponentPose=Local To Component(LocalPose=Copy Pose From Mesh() {bUseAttachedParent=true}), Alpha=1, ConstraintWeights_0=1) {
    alphaBoolBlend {blendOption=Linear}
    boneToModify {boneName=tongue_base}
    constraintSetup [{targetBone {boneName=head}}]
    constraintWeights [1]
   }
   n6 = Control Rig(Source=Component To Local(ComponentPose=@n5), Alpha=Get Jaw Open() -> (Jaw Open:double)) {
    alphaBoolBlend {blendOption=Linear}
    alphaScaleBiasClamp {bInterpResult=true}
    controlRigAssetReference {blueprintRigClass=&/M_LRG_Clash_V_FaceAcc_CtrlRig.M_LRG_Clash_V_FaceAcc_CtrlRig_C}
   }
   n7 = Local To Component(LocalPose=@n6)
   n8 = Get Tongue Scale() -> (Tongue Scale:Vector)
   Output Pose(Result=Component To Local(ComponentPose="Transform (Modify) Bone - Bone: tongue_base"!(ComponentPose=@n7, Alpha=1, Scale=@n8) -> (Pose) {scaleMode=BMM_Replace, alphaBoolBlend {blendOption=Linear}, boneToModify {boneName=tongue_base}})) {layerGroup=None}
  ]}
 EventGraph {nodes [
   Event BlueprintInitializeAnimation()
   > n1 = Cast To PlayerPawn_Athena(Object=TryGetPawnOwner())
   > n2 = Set Pawn Owner(Pawn Owner:Player Pawn Athena Object Reference=@n1) -> (Output_Get:Player Pawn Athena Object Reference)
   > Set Anim Instance(Anim Instance:Anim Instance Object Reference=GetAnimInstance(self=Get Mesh(self=@n2))) -> (Output_Get:Anim Instance Object Reference)
   > n3 = Set Head Mesh(Head Mesh:Skeletal Mesh Component Object Reference=GetSkeletalMeshForPartType(self=Get Pawn Owner() -> (Pawn Owner:Player Pawn Athena Object Reference), PartType=Head)) -> (Output_Get:Skeletal Mesh Component Object Reference)
   Event BlueprintUpdateAnimation()
   > n4 = Sequence!(execute) -> (then_0, then_1, then_2:Exec)
   n5 = GetOwnerFortPlayerAnimInstance!(self:Custom Character Part Anim Instance Object Reference) -> (ReturnValue:Fort Player Anim Instance Object Reference)
   > n6 = Is Valid(InputObject=@n5)
   Set Owner Playing Emote(execute=[@n7.then, @n6.Is Not Valid], Owner Playing Emote:bool=false) -> (Output_Get:bool)
   n7 = Branch(execute=@n6.Is Valid, Condition=Get bIsInFrontEnd())
   Set Owner Playing Emote(execute=@n7.else, Owner Playing Emote=Get bIsPlayingEmote!(self:Fort Player Anim Instance Object Reference=@n5) -> (bIsPlayingEmote))
   n8 = GetAlphaFromReferenceTransformDelta(SkeletalMeshComponent=@n3, BoneName=C_jaw, DeltaComponentType=QuaternionTwist, InputRange=(X=0,Y=15), TwistAxis="1,0,0")
   Set Jaw Open(execute=@n4.then_2, Jaw Open:double=MapRangeClamped(Value=@n8, InRangeA=1, InRangeB=0, OutRangeA=1, OutRangeB=0.2)) -> (Output_Get:double)
   n9 = MapRangeClamped(Value=@n8, InRangeA=1, InRangeB=0, OutRangeA=0.45, OutRangeB=0.2)
   n10 = MapRangeClamped(Value=@n8, InRangeA=0, InRangeB=0.5, OutRangeA=0.3)
   n11 = Get Owner Playing Emote() -> (Owner Playing Emote:bool)
   > Set Tongue Scale(Tongue Scale:Vector=MakeVector(X=SelectFloat(A=@n10, B=@n9, bPickA=@n11), Y=1, Z=1)) -> (Output_Get:Vector)
   n12 = GetAnimInstance(self=GetOwnerSkeletalMeshComponent())
   Is Valid(exec=@n4.then_1, InputObject=@n12)
   > n13 = Get Dynamic Material Instance(SkeletalMeshComponent=GetPartSkeletalMeshComponent())
   > SetScalarParameterValue(
    self=@n13
    ParameterName=Hide Element 10
    Value=SelectFloat(
     A=1
     B=0
     bPickA=Montage_IsActive!(
      self:Anim Instance Object Reference=@n12
      Montage:Anim Montage Object Reference=/BRCosmetics/Animation/Game/MainPlayer/Emotes/Vertigo/Emote_Vertigo_CMM_M
     ) -> (ReturnValue:bool)
    )
   )
   n14 = Montage_IsActive(
    self=GetAnimInstance(self=GetOwnerSkeletalMeshComponent())
    Montage=/BRCosmetics/Animation/Game/MainPlayer/Menu/BR/CustomSkins/_s18/Clash_V/Clash_V_Male_Idle_Montage
   )
   SelectFloat(A=1, B=0, bPickA=@n14)
   SelectFloat(A=1, B=0, bPickA=@n14)
  ]}
]
variables [
 Anim Instance {type=AnimInstance, default=null}
 Head Mesh {type=SkeletalMeshComponent}
 Jaw Open {type=number, default=1}
 Owner Playing Emote {type=boolean}
 Pawn Owner {type=PlayerPawn_Athena_C}
 Tongue Scale {type=Vector}
]
